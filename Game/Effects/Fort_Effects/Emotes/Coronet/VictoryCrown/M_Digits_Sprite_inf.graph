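# Masked, two-sided lit Niagara particle material for the Coronet Victory Crown emote that renders a numeric value as a row of digits.
# Dynamic parameters supply digit count and value; UVs split into one cell per digit, and power-of-ten division with modulo extracts each numeral.
# Each numeral offsets UVs into a 4x4 atlas sampled for digit shape and normals.
# A panning caustics highlight plus emissive color, tinted by particle color, feeds the emissive adjustment function.
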Material /Game/Effects/Fort_Effects/Emotes/Coronet/VictoryCrown/M_Digits_Sprite_inf
graph {
 properties {
  bCastRayTracedShadows=false
  bUsedWithNiagaraMeshParticles=true
  bUsedWithNiagaraSprites=true
  blendMode=BLEND_Masked
  pixelDepthOffsetMode=PDOM_Legacy
  twoSided=true
  usedShadingModels=Default Lit
 }
 nodes [
  Constant3Vector() {constant=<rgba 15 15 15 0>}
  n1 = ParticleColor()
  n2 = ScalarParameter() {defaultValue=0.1, parameterName=Highlight_Scale_U, sliderMax=1}
  n3 = ScalarParameter() {defaultValue=0.1, parameterName=Highlight_Scale_V, sliderMax=1}
  n4 = Panner(
   Coordinate=Multiply(A=TextureCoordinate(), B=AppendVector(A=@n2, B=@n3))
   Time=Time()
   Speed=AppendVector(
    A=ScalarParameter() {parameterName=Highlight_Pan_U, sliderMax=1, sliderMin=-1}
    B=ScalarParameter() {defaultValue=0.03, parameterName=Highlight_Pan_V, sliderMax=1, sliderMin=-1}
   )
  )
  n5 = Constant() {r=-2}
  n6 = Multiply(
   A=Power(Base=TextureSampleParameter2D!(UVs=@n4, Bias=@n5, Apply View MipBias) {mipValueMode=TMVM_MipBias, parameterName=HighlightTexture, texture=/FortWater/Textures/T_Caustics_3layer}.RGB) {constExponent=0.5}
   B=ScalarParameter() {defaultValue=1, parameterName=Highlight_Scale, sliderMax=1}
  )
  n7 = VectorParameter() {parameterName=Emissive}
  MaterialFunctionCall(Emissive=Multiply(A=@n1.RGB, B=Add(A=@n6, B=@n7.RGB))) {materialFunction=&MF_FX_EmissiveAdjustments}
  Multiply(A=@n1.RGB, B=VectorParameter() {parameterName=BaseColor, defaultValue=<rgba 15 15 15 0>}.RGB)
  StaticBool()
  n8 = MaterialFunctionCall(Float2=TextureCoordinate()) {materialFunction=&BreakOutFloat2Components}
  n9 = Multiply(
   A=OneMinus(None=@n8.R) {desc=inverse it... so we read left to right}
   B=DynamicParameter() {parameterIndex=1, paramNames [digits,value,Param3,Param4]}.digits
  )
  n10 = Constant() {r=1}
  n11 = MaterialFunctionCall(
   Input=OneMinus(None=Fmod(A=@n9, B=@n10)) {desc="re inverse it, otherwise numbers are backwards"}
   Input Low=Constant()
   Input High=Constant() {r=1}
   Target Low=Constant() {r=0.02}
   Target High=Constant() {r=0.98}
  ) {materialFunction=&RemapValueRange}
  n12 = Constant() {desc="4^4 texture", r=4}
  n13 = DynamicParameter() {parameterIndex=1, paramNames [digits,value,Param3,Param4]}
  n14 = Constant() {desc=Base, r=10}
  n15 = Power(Base=@n14, Exp=Floor(None=@n9) {desc="Digit position: 0,1,2..."}) {desc="1,10,100,1000..."}
  n16 = Fmod(A=Floor(None=Divide(A=@n13.value, B=@n15) {desc=Move digits to right}) {desc=Trim digits right of decimal}, B=@n14) {desc=Keep only rightmost digit}
  n17 = Floor(None=Fmod(A=@n16, B=@n12))
  n18 = Floor(None=Divide(A=@n16, B=@n12))
  n19 = Add(
   A=Divide(A=AppendVector(A=@n11.Result, B=@n8.G) {desc=make n digits uvs uvs}, B=@n12) {constB=4, desc=scale to texture sub uv size}
   B=Multiply(A=AppendVector(A=@n17, B=@n18), B=Divide(B=@n12)) {desc=normalize}
  ) {desc=offset}
  n20 = Constant() {r=-4}
  TextureSample!(UVs=@n19, Tex, Bias=@n20, Apply View MipBias) {
   mipValueMode=TMVM_MipBias
   samplerType=SAMPLERTYPE_LinearColor
   texture=&/EmoteCoronet_Type_Digits_A
  }
  TextureSample(UVs=@n19, Bias=@n20) {
   mipValueMode=TMVM_MipBias
   samplerType=SAMPLERTYPE_Normal
   texture=&/EmoteCoronet_Type_Digits_N
  }
  VectorParameter() {parameterName=Metallic, defaultValue=<rgba 1 0.6542 0 0>}
 ]
}
